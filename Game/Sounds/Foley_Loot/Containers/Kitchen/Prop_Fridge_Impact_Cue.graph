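# Sound cue for fridge prop impacts, set up with melee weapon sound class and small melee attenuation.
# A Random node picks one of three equally weighted fridge waves, which are named as search sounds despite the cue's impact name.
# Playback runs at reduced volume, and a cue-level concurrency override prevents new instances once the limit is reached.
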SoundCue /Game/Sounds/Foley_Loot/Containers/Kitchen/Prop_Fridge_Impact_Cue
root=Random_2
nodes [
 {id=Random_2, children [WavePlayer_10, WavePlayer_11, WavePlayer_12], properties {weights [1,1,1]}}
 {id=WavePlayer_10, properties {soundWaveAssetPtr=&/Prop_Fridge_Search_01}}
 {id=WavePlayer_11, properties {soundWaveAssetPtr=&/Prop_Fridge_Search_02}}
 {id=WavePlayer_12, properties {soundWaveAssetPtr=&/Prop_Fridge_Search_03}}
]
properties {
 attenuationSettings=&Fort_Weapon_Melee_Sml_Att
 bOverrideConcurrency=true
 maxDistance=2800
 soundClassObject=&Fort_WeaponMelee_PSM
 volumeMultiplier=0.5
 concurrencyOverrides {resolutionRule=PreventNew}
}
