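# Sound cue for fridge search completion, played as loot container foley in the Fort_Loot_Collect sound class.
# A mixer layers a random fridge search sound at reduced volume under a louder random gear search sound from the shared misc container set.
# An enveloper over the mix fades the volume out in under a second.
# Prop search attenuation applies, and a concurrency override allows one instance at a time, blocking new plays.
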SoundCue /Game/Sounds/Foley_Loot/Containers/Kitchen/Prop_Fridge_SearchComplete_Cue
root=Enveloper_6
nodes [
 {
  id=Enveloper_6
  children [Mixer_3]
  properties {pitchCurve {editorCurveData {keys [{time=0}]}}, volumeCurve {editorCurveData {keys [<at 0 1>,<at 0.64 0>]}}}
 }
 {id=Mixer_3, children [Random_1,Random_4], properties {inputVolume [0.3,1.2]}}
 {id=Random_1, children [WavePlayer_6,WavePlayer_7,WavePlayer_8], properties {weights [1,1,1]}}
 {id=Random_4, children [WavePlayer_21, WavePlayer_22, WavePlayer_24], properties {weights [1,1,1]}}
 {id=WavePlayer_21, properties {soundWaveAssetPtr=/Game/Sounds/Foley_Loot/Containers/Misc/Prop_Gear_Search_05}}
 {id=WavePlayer_22, properties {soundWaveAssetPtr=/Game/Sounds/Foley_Loot/Containers/Misc/Prop_Gear_Search_06}}
 {id=WavePlayer_24, properties {soundWaveAssetPtr=/Game/Sounds/Foley_Loot/Containers/Misc/Prop_Gear_Search_01}}
 {id=WavePlayer_6, properties {soundWaveAssetPtr=&/Prop_Fridge_Search_01}}
 {id=WavePlayer_7, properties {soundWaveAssetPtr=&/Prop_Fridge_Search_02}}
 {id=WavePlayer_8, properties {soundWaveAssetPtr=&/Prop_Fridge_Search_03}}
]
properties {
 attenuationSettings=/Game/Sounds/Attenuation/Fort_Prop_Search_Att
 bOverrideConcurrency=true
 maxDistance=1300
 soundClassObject=&Fort_Loot_Collect
 volumeMultiplier=0.7
 concurrencyOverrides {maxCount=1, resolutionRule=PreventNew}
}
