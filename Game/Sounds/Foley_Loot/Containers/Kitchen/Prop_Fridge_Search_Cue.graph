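# Sound cue for the foley heard when searching a fridge loot container in the kitchen prop set.
# A mixer layers a random pick of three fridge search waves, at reduced input volume, with a random pick of three generic gear search waves at full volume.
# Both layers share the same short fixed delay.
# The cue uses prop search attenuation, the loot collect sound class and a concurrency override that prevents new instances at the limit.
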SoundCue /Game/Sounds/Foley_Loot/Containers/Kitchen/Prop_Fridge_Search_Cue
root=Mixer_0
nodes [
 {id=Delay_4, children [Random_2], properties {delayMax=0.1, delayMin=0.1}}
 {id=Delay_5, children [Random_0], properties {delayMax=0.1, delayMin=0.1}}
 {id=Mixer_0, children [Delay_4,Delay_5], properties {inputVolume [0.4,1]}}
 {id=Random_0, children [WavePlayer_1,WavePlayer_2,WavePlayer_4], properties {weights [1,1,1]}}
 {id=Random_2, children [WavePlayer_10, WavePlayer_11, WavePlayer_12], properties {weights [1,1,1]}}
 {id=WavePlayer_1, properties {soundWaveAssetPtr=/Game/Sounds/Foley_Loot/Containers/Misc/Prop_Gear_Search_05}}
 {id=WavePlayer_10, properties {soundWaveAssetPtr=&/Prop_Fridge_Search_01}}
 {id=WavePlayer_11, properties {soundWaveAssetPtr=&/Prop_Fridge_Search_02}}
 {id=WavePlayer_12, properties {soundWaveAssetPtr=&/Prop_Fridge_Search_03}}
 {id=WavePlayer_2, properties {soundWaveAssetPtr=/Game/Sounds/Foley_Loot/Containers/Misc/Prop_Gear_Search_06}}
 {id=WavePlayer_4, properties {soundWaveAssetPtr=/Game/Sounds/Foley_Loot/Containers/Misc/Prop_Gear_Search_01}}
]
properties {
 attenuationSettings=/Game/Sounds/Attenuation/Fort_Prop_Search_Att
 bOverrideConcurrency=true
 maxDistance=1300
 soundClassObject=&Fort_Loot_Collect
 volumeMultiplier=0.7
 concurrencyOverrides {resolutionRule=PreventNew}
}
