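# Sound cue for a kitchen grill container impact, built from stove and phone booth sounds and using small melee weapon attenuation.
# A quality switch sends the first two levels to a mix of a phone booth impact and one of two random stove search sounds.
# The other two levels play a single optimized stove impact with slight pitch variation.
# An outer modulator pitches everything up, and a concurrency override prevents new instances at the limit.
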SoundCue /Game/Sounds/Foley_Loot/Containers/Kitchen/Prop_Grill_Impact_Cue
root=Modulator_0
nodes [
 {id=Mixer_1, children [WavePlayer_5,Random_1], properties {inputVolume [0.8,1]}}
 {id=Modulator_0, children [QualityLevel_0], properties {pitchMax=1.5, pitchMin=1.3}}
 {id=Modulator_1, children [WavePlayer_0], properties {pitchMax=1.1, pitchMin=0.9}}
 {id=QualityLevel_0, children [Mixer_1,Mixer_1,Modulator_1,Modulator_1]}
 {id=Random_1, children [WavePlayer_3,WavePlayer_4], properties {weights [1,1]}}
 {id=WavePlayer_0, properties {soundWaveAssetPtr=&/Prop_Stove_Impact_opt}}
 {id=WavePlayer_3, properties {soundWaveAssetPtr=&/Prop_Stove_Search_02}}
 {id=WavePlayer_4, properties {soundWaveAssetPtr=&/Prop_Stove_Search_06}}
 {id=WavePlayer_5, properties {soundWaveAssetPtr=/Game/Sounds/Fort_Impact_Sounds/Demolition/PhoneBooth/phonebooth_med_impact_01}}
]
properties {
 attenuationSettings=&Fort_Weapon_Melee_Sml_Att
 bOverrideConcurrency=true
 maxDistance=2800
 soundClassObject=&Fort_WeaponMelee_PSM
 volumeMultiplier=0.5
 concurrencyOverrides {resolutionRule=PreventNew}
}
