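# Sound cue for searching a kitchen stove hood container, built as layered loot foley.
# A mixer blends a quieter random pick of three stove search waves, pitched down by a fixed modulator, with a louder random pick of three generic gear search waves.
# It uses the prop search attenuation and loot collect sound class, with a concurrency override that prevents new instances once the limit is reached.
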SoundCue /Game/Sounds/Foley_Loot/Containers/Kitchen/Prop_StoveHood_Search_Cue
root=Mixer_2
nodes [
 {id=Mixer_2, children [Modulator_5,Random_5], properties {inputVolume [0.3,1.2]}}
 {id=Modulator_5, children [Random_3], properties {pitchMax=0.7, pitchMin=0.7}}
 {id=Random_3, children [WavePlayer_22, WavePlayer_23, WavePlayer_24], properties {weights [1,1,1]}}
 {id=Random_5, children [WavePlayer_11, WavePlayer_12, WavePlayer_14], properties {weights [1,1,1]}}
 {id=WavePlayer_11, properties {soundWaveAssetPtr=/Game/Sounds/Foley_Loot/Containers/Misc/Prop_Gear_Search_05}}
 {id=WavePlayer_12, properties {soundWaveAssetPtr=/Game/Sounds/Foley_Loot/Containers/Misc/Prop_Gear_Search_06}}
 {id=WavePlayer_14, properties {soundWaveAssetPtr=/Game/Sounds/Foley_Loot/Containers/Misc/Prop_Gear_Search_01}}
 {id=WavePlayer_22, properties {soundWaveAssetPtr=&/Prop_Stove_Search_03}}
 {id=WavePlayer_23, properties {soundWaveAssetPtr=&/Prop_Stove_Search_04}}
 {id=WavePlayer_24, properties {soundWaveAssetPtr=&/Prop_Stove_Search_05}}
]
properties {
 attenuationSettings=/Game/Sounds/Attenuation/Fort_Prop_Search_Att
 bOverrideConcurrency=true
 maxDistance=1300
 soundClassObject=&Fort_Loot_Collect
 volumeMultiplier=1
 concurrencyOverrides {resolutionRule=PreventNew}
}
