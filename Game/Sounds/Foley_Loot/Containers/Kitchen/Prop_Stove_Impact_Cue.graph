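# Sound cue for the impact sound of a kitchen stove prop, using small melee weapon attenuation, sound class and concurrency.
# A quality level node sends the first two levels to a mixer layering a phone booth impact with one of two random stove search sounds.
# The last two levels play a single "opt" stove impact wave through slight random pitch modulation.
# Concurrency overrides prevent new instances at the limit.
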SoundCue /Game/Sounds/Foley_Loot/Containers/Kitchen/Prop_Stove_Impact_Cue
root=QualityLevel_0
nodes [
 {id=Mixer_0, children [WavePlayer_2,Random_0], properties {inputVolume [0.8,1]}}
 {id=Modulator_0, children [WavePlayer_3], properties {pitchMax=1.1, pitchMin=0.9}}
 {id=QualityLevel_0, children [Mixer_0,Mixer_0,Modulator_0,Modulator_0]}
 {id=Random_0, children [WavePlayer_0,WavePlayer_1], properties {weights [1,1]}}
 {id=WavePlayer_0, properties {soundWaveAssetPtr=&/Prop_Stove_Search_02}}
 {id=WavePlayer_1, properties {soundWaveAssetPtr=&/Prop_Stove_Search_06}}
 {id=WavePlayer_2, properties {soundWaveAssetPtr=/Game/Sounds/Fort_Impact_Sounds/Demolition/PhoneBooth/phonebooth_med_impact_01}}
 {id=WavePlayer_3, properties {soundWaveAssetPtr=&/Prop_Stove_Impact_opt}}
]
properties {
 attenuationSettings=&Fort_Weapon_Melee_Sml_Att
 maxDistance=2800
 soundClassObject=&Fort_WeaponMelee_PSM
 volumeMultiplier=0.5
 concurrencyOverrides {resolutionRule=PreventNew}
 concurrencySet [&Player_Melee_Weapon_Conc]
}
