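# Sound cue for finishing a kitchen stove container search, pairing a "no more loot" UI sound with a short foley burst.
# The foley mixes one random stove search wave with a louder random gear search wave, shaped by an Enveloper that rises quickly and fades out in under a second.
# It uses prop-search attenuation and the loot-collect sound class, with a concurrency override allowing one instance and blocking new ones.
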SoundCue /Game/Sounds/Foley_Loot/Containers/Kitchen/Prop_Stove_SearchDone_Cue
root=Mixer_1
nodes [
 {
  id=Enveloper_5
  children [Mixer_4]
  properties {pitchCurve {editorCurveData {keys [{time=0}]}}, volumeCurve {editorCurveData {keys [<at 0 0>,<at 0.0542 0.8454>,<at 0.64 0>]}}}
 }
 {id=Mixer_1, children [WavePlayer_12, Enveloper_5], properties {inputVolume [0.8,1]}}
 {id=Mixer_4, children [Random_2,Random_6], properties {inputVolume [0.7,1.5]}}
 {id=Random_2, children [WavePlayer_16, WavePlayer_17, WavePlayer_18], properties {weights [1,1,1]}}
 {id=Random_6, children [WavePlayer_31, WavePlayer_32, WavePlayer_34], properties {weights [1,1,1]}}
 {id=WavePlayer_12, properties {soundWaveAssetPtr=/Game/Sounds/Fort_UI_Sounds/Gameplay/Fort_UI_NoMoreLoot}}
 {id=WavePlayer_16, properties {soundWaveAssetPtr=&/Prop_Stove_Search_03}}
 {id=WavePlayer_17, properties {soundWaveAssetPtr=&/Prop_Stove_Search_04}}
 {id=WavePlayer_18, properties {soundWaveAssetPtr=&/Prop_Stove_Search_05}}
 {id=WavePlayer_31, properties {soundWaveAssetPtr=/Game/Sounds/Foley_Loot/Containers/Misc/Prop_Gear_Search_05}}
 {id=WavePlayer_32, properties {soundWaveAssetPtr=/Game/Sounds/Foley_Loot/Containers/Misc/Prop_Gear_Search_06}}
 {id=WavePlayer_34, properties {soundWaveAssetPtr=/Game/Sounds/Foley_Loot/Containers/Misc/Prop_Gear_Search_01}}
]
properties {
 attenuationSettings=/Game/Sounds/Attenuation/Fort_Prop_Search_Att
 bOverrideConcurrency=true
 maxDistance=1300
 soundClassObject=&Fort_Loot_Collect
 volumeMultiplier=0.7
 concurrencyOverrides {maxCount=1, resolutionRule=PreventNew}
}
